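# UI material for a Lucky-themed Item Shop tile, drawing animated clovers over green radial gradients.
# A particle grid drifts randomly scaled and rotated clover texture cells, whose green channel blends the background toward a second gradient.
# A time-driven diagonal shine, gated by a slow sine on the `tenminutes` parameter, brightens the clovers.
# Separate branches build a twinkle mask, gold tint and rotating pattern UVs, but none feeds that blend.
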Material /ItemShop/UI/Art/Lucky/M_UI_Lucky_ShopTile
graph {
 properties {bUsedWithStaticMesh=false, materialDomain=MD_UI, pixelDepthOffsetMode=PDOM_Legacy, usedShadingModels=Default Lit}
 nodes [
  n1 = Time()
  n2 = ScalarParameter() {defaultValue=0.5, parameterName=tenminutes}
  n3 = Multiply(A=Sine(None=Multiply(A=@n1, B=@n2)) {period=3}) {constB=3}
  Ceil(None=@n3)
  n4 = LinearInterpolate(
   A=Constant3Vector() {constant=<rgba 0.0028 0.0955 0 1>}
   B=Constant3Vector() {constant=<rgba 0 0.6372 0.1029 1>}
   Alpha=MaterialFunctionCall(Radius=Constant() {r=0.9}, Density=Constant() {r=1}) {materialFunction=&RadialGradientExponential}.RadialGradientExponential
  )
  n5 = LinearInterpolate(
   A=Constant3Vector() {constant=<rgba 0 0.0521 0.0195 1>}
   B=Constant3Vector() {constant=<rgba 0 0.8698 0.1788 1>}
   Alpha=MaterialFunctionCall(Radius=Constant() {r=0.65}, Density=Constant() {r=1.25}) {materialFunction=&RadialGradientExponential}.RadialGradientExponential
  )
  n6 = Constant3Vector() {constant=<rgba 0.0135 0.5191 0.1175 1>}
  n7 = SmoothStep(Value=@n3) {constMax=8, constMin=2}
  n8 = Frac(None=OneMinus(None=Time()))
  n9 = Add(
   A=ConstantBiasScale(None=MaterialFunctionCall() {materialFunction=&MF_UI_UVScaling}.Scaled To Fill) {scale=0.24}
   B=AppendVector(A=Multiply(A=@n8), B=Constant())
  )
  n10 = Constant() {r=0.5}
  n11 = Constant() {r=0.35}
  n12 = ComponentMask(None=Subtract(A=MaterialFunctionCall(UVs=@n9, Rotation Center=@n10, "Rotation Angle (0-1)"=@n11) {materialFunction=&CustomRotator}.Rotated Values) {constB=0}) {r=true}
  n13 = Constant() {r=0.005}
  n14 = Subtract(A=Sine(None=Multiply(A=@n12) {constB=0.5}), B=OneMinus(None=@n13)) {constB=0.25}
  n15 = Divide(B=@n13) {constB=0.5}
  n16 = Saturate(None=Multiply(A=Multiply(A=@n14, B=@n15)) {constB=2})
  n17 = Multiply(A=Multiply(A=Saturate(None=@n7), B=Multiply(A=@n16) {constB=10})) {constB=0.5}
  n18 = NamedRerouteUsage()
  n19 = Frac(None=Multiply(A=Time()) {constB=0.015})
  n20 = MaterialFunctionCall!(
   UV=MaterialFunctionCall() {materialFunction=&MF_UI_UVScaling}.Scaled to Fit
   Global Offset=AppendVector(A=Add(A=@n18, B=@n19), B=Constant())
   Global Rotation=Constant() {r=0.35}
   Tiling=Constant() {r=6}
   Cell Offset=Constant()
   Cell Scale=LinearInterpolate(Alpha=NamedRerouteUsage()) {constA=0.25, constB=1.1}
   Cell Rotation=Add(A=NamedRerouteUsage(), B=@n19)
   Random Seed
  ) {materialFunction=/FortUILibrary/MaterialFunctions/Basics/MF_UI_ParticleGrid}
  n21 = TextureSample(UVs=@n20.Cells UV) {samplerType=SAMPLERTYPE_Alpha, texture=/Game/UI/Frontend/ItemShop/Art/Lucky/T_UI_LuckyClover-01}
  LinearInterpolate(A=@n4, B=LinearInterpolate(A=@n4, B=LinearInterpolate(A=@n5, B=@n6, Alpha=@n17), Alpha=@n21.G), Alpha=@n21.G)
  n22 = MaterialFunctionCall(
   UVs=MaterialFunctionCall() {materialFunction=&MF_UI_UVScaling}.Scaled to Fit
   Texture Scale=Constant() {r=2.023}
  ) {materialFunction=&ScaleUVsByCenter}
  n23 = MaterialFunctionCall!(
   UV=MaterialFunctionCall(UVs=Panner(Coordinate=@n22.UVs) {speedX=-0.005, speedY=-0.007}, "Rotation Angle (0-1)"=Constant() {r=0.35}) {materialFunction=&CustomRotator}.Rotated Values
   Tiling X=Constant() {desc=num stars in a row, r=7}
   Tiling Y=Constant() {r=7}
  ) {materialFunction=/FortUILibrary/MaterialFunctions/Basics/MF_UI_Pattern}
  n24 = MaterialFunctionCall!(UV=Floor(None=@n23.Raw UVs), Seed=Constant() {r=0.05}) {materialFunction=&MF_UI_Random}
  n25 = Add(A=@n24.Result)
  n26 = Multiply(A=@n23.Scaled Pattern UVs, B=@n25)
  n27 = Multiply(A=Time()) {constB=0.5}
  n28 = Multiply(A=@n24.Result) {constB=1.7}
  MaterialFunctionCall(
   UVs=MaterialFunctionCall(UVs=Rotator(Coordinate=@n26, Time=@n27), Texture Scale=Max(A=@n28) {constB=0.5}) {materialFunction=&ScaleUVsByCenter}.UVs
   "Rotation Angle (0-1)"=LinearInterpolate(Alpha=@n25) {constB=0.5}
  ) {materialFunction=&CustomRotator}
  n29 = NamedRerouteUsage()
  n30 = Frac(None=Multiply(A=Time()) {constB=0.5})
  n31 = Min(A=@n30, B=OneMinus(None=@n30))
  n32 = Subtract(A=@n29, B=LinearInterpolate(Alpha=Multiply(A=@n31) {constB=2}) {constB=0.75})
  n33 = Add(A=@n29) {constB=0.2}
  Multiply(A=Multiply(A=@n21.G, B=Saturate(None=@n32)), B=Round(None=Saturate(None=@n33)))
  Multiply(
   A=MaterialFunctionCall(UV=@n21.R, GradientStart=Constant(), GradientEnd=Constant() {r=1.5}) {materialFunction=&MF_UI_RadialGradient}.Smooth
   B=Constant3Vector() {constant=<rgba 0.8594 0.534 0 1>}
  )
  NamedRerouteDeclaration(None=@n20.Random Rows) {name=Random Rows, nodeColor=<rgba 0.7344 0 1 1>}
  NamedRerouteDeclaration(None=@n20.Random Cell Values) {name=Random Cell Values, nodeColor=<rgba 0.7344 0 1 1>}
  NamedRerouteDeclaration(None=@n20.Random Columns) {name=Columns, nodeColor=<rgba 1 0.4688 0 1>}
 ]
}
